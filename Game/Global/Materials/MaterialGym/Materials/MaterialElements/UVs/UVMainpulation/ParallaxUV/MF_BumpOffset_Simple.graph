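# Material function that produces parallax-shifted UVs from the camera angle, giving a simple bump offset without a height map.
# It normalizes the world-space camera-to-pixel difference (excluding shader offsets), transforms it to tangent space and keeps only XY.
# That vector, scaled by the BumpOffsetIntensity parameter, is subtracted from the texture coordinates and output as BumpOffset_SimpleUV.
# With no divide by Z, the shift stays bounded at grazing angles.
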MaterialFunction /Game/Global/Materials/MaterialGym/Materials/MaterialElements/UVs/UVMainpulation/ParallaxUV/MF_BumpOffset_Simple
graph {nodes [
  n1 = Subtract(A=CameraPositionWS(), B=WorldPosition() {worldPositionShaderOffset=WPT_ExcludeAllShaderOffsets}.XYZ)
  n2 = ComponentMask(None=Transform(None=Normalize(VectorInput=@n1)) {transformSourceType=TRANSFORMSOURCE_World, transformType=TRANSFORM_Tangent}) {g=true, r=true}
  n3 = ScalarParameter() {defaultValue=7.5662, group=BumpOffset_UV, parameterName=BumpOffsetIntensity}
  FunctionOutput(None=Subtract(A=TextureCoordinate(), B=Multiply(A=@n2, B=@n3))) {outputName=BumpOffset_SimpleUV}
 ]}
