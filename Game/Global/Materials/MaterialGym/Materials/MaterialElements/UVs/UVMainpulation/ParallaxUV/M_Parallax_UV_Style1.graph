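# M_Parallax_UV_Style1 is a Material Gym example material that shows parallax UV manipulation.
# It calls the MF_BumpOffset_Simple material function with its default inputs to produce offset UVs.
# Those UVs drive a UVTexture parameter sample of a color grid texture, making the distortion visible, using a shared world-wrap sampler.
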Material /Game/Global/Materials/MaterialGym/Materials/MaterialElements/UVs/UVMainpulation/ParallaxUV/M_Parallax_UV_Style1
graph {
 properties {bAutomaticallySetUsageInEditor=false, usedShadingModels=Default Lit}
 nodes [TextureSampleParameter2D(UVs=MaterialFunctionCall() {materialFunction=&/MF_BumpOffset_Simple}.BumpOffset_SimpleUV) {group=BumpOffset_UV, parameterName=UVTexture, samplerSource=SSM_Wrap_WorldGroupSettings, texture=/UVEditor/Textures/UVEditorColorGrid_Color}]
}
