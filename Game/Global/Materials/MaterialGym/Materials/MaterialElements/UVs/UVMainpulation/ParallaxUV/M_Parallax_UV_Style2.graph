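# Material gym sample, Style 2 of parallax UV manipulation, demonstrating bump offset on a UV checker grid texture.
# Texture coordinates and a small constant height ratio feed MF_BumpOffset_Accurate, leaving Height, ReferencePlane and Normal unconnected.
# The offset UVs pass through a Multiply with no second input into the UVTexture parameter, which uses the shared wrap sampler.
# The graph shows no connection from that sample to a material output.
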Material /Game/Global/Materials/MaterialGym/Materials/MaterialElements/UVs/UVMainpulation/ParallaxUV/M_Parallax_UV_Style2
graph {
 properties {bAutomaticallySetUsageInEditor=false, usedShadingModels=Default Lit}
 nodes [
  n1 = TextureCoordinate()
  n2 = Constant() {r=0.01}
  TextureSampleParameter2D(UVs=Multiply(A=MaterialFunctionCall!(UV Coord=@n1, Height, ReferencePlane, HeightRatio=@n2, Normal) {materialFunction=/Game/Global/MaterialFunctions/MF_BumpOffset_Accurate}.Result)) {group=BumpOffset_UV, parameterName=UVTexture, samplerSource=SSM_Wrap_WorldGroupSettings, texture=/UVEditor/Textures/UVEditorColorGrid_Color}
 ]
}
